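# Niagara system for face sweat on the Relaxed emote, built from a single Sweat_Drops sprite emitter.
# Spawns short-lived drops at a random rate in a flat, offset box, then launches them with random speed under gravity and drag.
# Size curves narrow and lengthen each drop over its life to fake streaking.
# A curve-driven dynamic material parameter and a faint alpha that fades late in life shape the look.
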NiagaraSystem /BRCosmetics/Animation/Game/MainPlayer/Emotes/Scamper/FX/NS_Emote_Relaxed_FaceSweat
dependencies {
 usedDataInterfaces [&NiagaraDataInterfaceVector2DCurve, &NiagaraDataInterfaceCurve]
 usedDynamicInputs []
 usedModules [&SystemState]
}
emitters {Sweat_Drops {
  emitterUpdateScript [
   &EmitterState(Inactive Response=::NewEnumerator1, Life Cycle Mode=::NewEnumerator1, Loop Duration=2, UseLoopDelay=false)
   &SpawnRate(SpawnRate=&RandomRangeFloat(Minimum=10, Maximum=20), Use Spawn Probability=false)
   &SpawnRate(SpawnRate=500, Use Spawn Probability=false) {enabled=false, module=SpawnRate001}
  ]
  particleSpawnScript [
   &InitializeParticle(
    Lifetime Max=0.4
    Lifetime Min=0.25
    Lifetime Mode=::NewEnumerator1
    Mass Mode=::NewEnumerator3
    Sprite Size Max=<1.25 4>
    Sprite Size Min=<0.8 3>
    Sprite Size Mode=::NewEnumerator2
    UsePositionOffset=false
    Write Material Random=true
   )
   &ShapeLocation(
    Hemisphere Distribution=<0.5 1>
    MaskSpawnGroup=false
    Offset=<0 0 5>
    Offset Mode=::NewEnumerator0
    Override Seed=false
    Sphere Radius=6
    Sphere Surface Distribution=1
    UseOwnerScale=true
   ) {enabled=false}
   &ShapeLocation(
    Box Size=<0 10 20>
    MaskSpawnGroup=false
    Offset=<12 3 0>
    Offset Mode=::NewEnumerator0
    Override Seed=false
    Shape Primitive=::NewEnumerator2
    Surface Only=false
    UseOwnerScale=true
   ) {module=ShapeLocation001}
   &AddVelocity(
    Module.Use Falloff Distance=false
    Override Seed=false
    Velocity Mode=::NewEnumerator1
    Velocity Speed=&RandomRangeFloat(Minimum=50, Maximum=100)
   )
   &AddVelocity(Velocity=&RandomRangeVector(Minimum=<0 0 -100>, Maximum=<0 0 100>)) {enabled=false, module=AddVelocity001}
   &ApplyInitialForces()
  ]
  particleUpdateScript [
   &ParticleState()
   &GravityForce()
   &ScaleSpriteSize(Scale Factor=&Vector2DFromCurve(
     Vector2Curve=DataInterface() {propertyValues {
       XCurve {keys [<at 0.25 1 RCIM_Cubic>, {interpMode=RCIM_Cubic, leaveTangent=-2.9373, time=1, value=0.3}]}
       YCurve {keys [
         {arriveTangent=-1.4216, interpMode=RCIM_Cubic, time=0.0067, value=0.3562}
         {arriveTangent=2.9367, interpMode=RCIM_Cubic, leaveTangent=2.9367, tangentMode=RCTM_User, time=0.2953, value=1.3705}
         {interpMode=RCIM_Cubic, leaveTangent=7.4188, time=1, value=2}
        ]}
      }}
     ScaleCurve=<1 1>
    ))
   &Drag(Drag=0.5, Use Linear Drag=true, Use Rotational Drag=false)
   &DynamicMaterialParameters(
    Index 0 Param 1=0.8
    Index 0 Param 2=&FloatFromCurve(FloatCurve=DataInterface() {propertyValues {Curve {keys [{arriveTangent=-1.8884, interpMode=RCIM_Cubic, leaveTangent=-1.8884, tangentMode=RCTM_User, time=0, value=1}, <at 1 0.26>]}}})
    Index 0 Param 3=1
    Index 0 Param 4=1
    Module.Param0WriteEnabled=true
    Module.Param1WriteEnabled=true
    Module.Param2WriteEnabled=false
    Module.Param3WriteEnabled=false
   )
   &ScaleColor(
    Scale Alpha=&FloatFromCurve(
     FloatCurve=DataInterface() {propertyValues {Curve {keys [<at 0 1>, {arriveTangent=-1, interpMode=RCIM_Cubic, leaveTangent=-1, time=0.8, value=1}, <at 1 0>]}}}
     Scale Curve=&RandomRangeFloat(Minimum=0.1, Maximum=0.2)
    )
    ScaleA=true
    ScaleRGB=false
   ) {module=ScaleColor001}
   &SolveForcesAndVelocity(Clamp Velocity=false, Limit Acceleration=false)
  ]
  rendererClasses [&NiagaraSpriteRendererProperties]
  renderers [{rendererClass=&NiagaraSpriteRendererProperties}]
 }}
summary {
 systemName=NS_Emote_Relaxed_FaceSweat
 emitters [{emitterName=Sweat_Drops, rendererClasses [&NiagaraSpriteRendererProperties]}]
 userVariables [User.bIsFrontEndPreview {type=NiagaraBool}]
}
